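; Construction script for PBWA_W1_StairSpiral, a player-built level 1 wood spiral stair actor in the building system.
; It gets the actor's static mesh component and sets material slot 0 to the static Wood_lvl1 material from the PBW wood package.
; No other logic appears in the file.
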(fn ConstructionScript
  (Rendering|Material|SetMaterial (Variables|Debug|GetStaticMeshComponent) 0 "/Game/Packages/PBW/Wood/Materials/Wood_lvl1_Static.Wood_lvl1_Static"))
